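# Translucent UI material for Item Shop tile backgrounds that draws a rotated egg pattern over a colored gradient, with an animated sheen.
# Shared background parameters supply the colors, and a static switch picks a linear or radial gradient.
# The pattern mask uses screen-space scaled UVs, with parameters controlling how it blends with the gradient.
# A delayed, angled sheen band brightens the result, optionally limited to the pattern, alongside a shared rounded-corners function.
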Material /Game/Catalog/BackgroundMaterialLibrary/CoreMaterials/Spring_Egg/M_UI_Shop_Background_Egg
graph {
 properties {bAutomaticallySetUsageInEditor=false, blendMode=BLEND_Translucent, materialDomain=MD_UI, usedShadingModels=Default Lit}
 nodes [
  n1 = NamedRerouteUsage()
  n2 = NamedRerouteUsage()
  LinearInterpolate(
   A=@n1
   B=Saturate(None=MaterialFunctionCall!(Base=@n1, Add=@n2) {materialFunction=/Engine/Functions/Engine_MaterialFunctions02/Math/MultiplyAdd}.Result)
   Alpha=LinearInterpolate(B=NamedRerouteUsage(), Alpha=ScalarParameter() {defaultValue=1, group=04 - Sheen, parameterName=Apply Sheen on Pattern, sliderMax=1}) {constA=1}
  )
  MaterialFunctionCall() {materialFunction=&MF_UI_UVScaling}
  MaterialFunctionCall() {materialFunction=&GetUserInterfaceUV}
  MaterialFunctionCall!(Input Alpha=Constant() {r=1}, bOverrideCornerRadius, Corner Radius Override) {materialFunction=/FortUILibrary/Systems/ItemShopTile/Materials/MaterialFunctions/MF_UI_ItemShop_RoundedCorners}
  n3 = MaterialFunctionCall() {materialFunction=/FortUILibrary/Systems/ItemShopTile/Materials/MaterialFunctions/MF_UI_ItemShop_BackgroundParams}
  NamedRerouteDeclaration(None=@n3.Color 3) {name=Accent Color, nodeColor=<rgba 0 0 0 1>}
  NamedRerouteDeclaration(None=@n3.Color 1) {name=Primary Color, nodeColor=<rgba 0 0 0 1>}
  n4 = NamedRerouteUsage()
  n5 = ScalarParameter() {group=01 - Gradient, parameterName=Gradient Rotation}
  n6 = Add(
   A=MaterialFunctionCall(UVs=NamedRerouteUsage(), Texture Scale=ScalarParameter() {defaultValue=1, group=01 - Gradient, parameterName=Radial Gradient Scale}) {materialFunction=&ScaleUVsByCenter}.UVs
   B=AppendVector(
    A=ScalarParameter() {group=01 - Gradient, parameterName=Radial Gradient Offset X}
    B=ScalarParameter() {group=01 - Gradient, parameterName=Radial Gradient Offset Y}
   )
  ) {constB=0.5}
  n7 = StaticSwitchParameter(
   True=SmoothStep(Value=MaterialFunctionCall(UVs=@n4, RotationAngle=@n5) {materialFunction=&MF_UI_LinearGradient}.Result)
   False=MaterialFunctionCall(UV=MaterialFunctionCall(UV=@n6) {materialFunction=&MF_UI_UVScaling}.Scaled To Fill) {materialFunction=&MF_UI_RadialGradient}.Smooth
  ) {group=01 - Gradient, parameterName=Use Linear Gradient, sortPriority=-32}
  NamedRerouteDeclaration(None=LinearInterpolate(A=NamedRerouteUsage(), B=NamedRerouteUsage(), Alpha=Saturate(None=@n7))) {name=Colored Gradient, nodeColor=<rgba 0 0 0 1>}
  n8 = MaterialFunctionCall() {materialFunction=&MF_UI_UVScaling}
  NamedRerouteDeclaration(None=@n8.Scaled To Fill) {name=Normalized UV, nodeColor=<rgba 0 0 0 1>}
  n9 = MaterialFunctionCall!(ScreenUV, TextureSize, ScalePivotPoint) {materialFunction=/FortUILibrary/MaterialFunctions/Basics/MF_UI_ScreenSpaceUVScaling}
  n10 = Constant() {r=0.2}
  n11 = TextureObjectParameter() {
   group=02 - Pattern
   parameterName=Pattern Texture
   samplerType=SAMPLERTYPE_Alpha
   texture=/Game/Catalog/BackgroundMaterialLibrary/Textures/T_UI_ShopTile_Egg
  }
  n12 = TextureSample(
   UVs=MaterialFunctionCall(
    UVs=MaterialFunctionCall(UVs=@n9.Scaled to Fill, Texture Scale=@n10) {materialFunction=&ScaleUVsByCenter}.UVs
    Rotation Center=Constant() {r=0.5}
    "Rotation Angle (0-1)"=Constant() {r=0.05}
   ) {materialFunction=&CustomRotator}.Rotated Values
   Tex=@n11
  )
  NamedRerouteDeclaration(None=Multiply(A=NamedRerouteDeclaration(None=@n12.R) {name=Pattern Mask, nodeColor=<rgba 0 0 0 1>}) {constB=0.5}) {name=Pattern Opacity, nodeColor=<rgba 0 0 0 1>}
  n13 = NamedRerouteUsage()
  n14 = ScalarParameter() {group=04 - Sheen, parameterName=Sheen Angle}
  n15 = LinearInterpolate(Alpha=MaterialFunctionCall!(
    Animation Duration=ScalarParameter() {defaultValue=1, group=04 - Sheen, parameterName=Sheen Duration}
    Delay=ScalarParameter() {defaultValue=3, group=04 - Sheen, parameterName=Sheen Delay}
    Reset to Default
   ) {materialFunction=/FortUILibrary/MaterialFunctions/Time/MF_UI_AnimationDelay}.Result) {constA=-1}
  n16 = Constant()
  n17 = MaterialFunctionCall(
   UVs=Add(A=Rotator(Coordinate=@n13, Time=@n14) {speed=6.2832}, B=AppendVector(A=@n15, B=@n16))
   Texture Scale=AppendVector(A=ScalarParameter() {defaultValue=0.2, group=04 - Sheen, parameterName=Sheen Scale}, B=Constant() {r=1})
  ) {materialFunction=&ScaleUVsByCenter}
  n18 = ComponentMask(None=MaterialFunctionCall!(Value=@n17.UVs, Scale=Constant() {r=1}, Height=Constant() {r=1}) {materialFunction=/FortUILibrary/MaterialFunctions/Basics/MF_UI_PingPong}.Result) {r=true}
  NamedRerouteDeclaration(None=Multiply(
    A=MaterialFunctionCall!(UVs=@n17.UVs, Alpha=@n18) {materialFunction=/FortUILibrary/MaterialFunctions/Basics/MF_UI_ClipEdgePixel}.Result
    B=ScalarParameter() {defaultValue=1, group=04 - Sheen, parameterName=Sheen Opacity, sliderMax=1}
   )) {name=Sheen Opacity, nodeColor=<rgba 0 0 0 1>}
  n19 = NamedRerouteUsage()
  n20 = NamedRerouteUsage()
  n21 = ScalarParameter() {defaultValue=1, group=02 - Pattern, parameterName=Pattern Blend, sliderMax=1, sliderMin=-1}
  n22 = NamedRerouteUsage()
  n23 = Saturate(None=MaterialFunctionCall(Base=@n19, Add=Multiply(A=@n21, B=@n22) {constB=2}) {materialFunction=&MultiplyAdd}.Result)
  n24 = ScalarParameter() {group=02 - Pattern, parameterName=Blend Pattern and Gradient, sliderMax=1}
  NamedRerouteDeclaration(None=LinearInterpolate(A=@n19, B=LinearInterpolate(A=@n20, B=@n23, Alpha=@n24), Alpha=NamedRerouteUsage())) {name=Pattern Color, nodeColor=<rgba 0 0 0 1>}
  n25 = ScalarParameter() {defaultValue=1, group=0 - Global, parameterName=Global Speed}
  NamedRerouteDeclaration(None=@n25) {name=Global Speed Input, nodeColor=<rgba 0 0 0 1>}
  NamedRerouteDeclaration(None=@n8.Scaled To Fill) {name=Scale to Fill UV, nodeColor=<rgba 0 0 0 1>}
  NamedRerouteDeclaration(None=@n3.Color 2) {name=Secondary Color, nodeColor=<rgba 0 0 0 1>}
  NamedRerouteDeclaration(None=@n8.ScaleY) {name=Scale Y UV, nodeColor=<rgba 0 0 0 1>}
  NamedRerouteDeclaration(None=Multiply(A=Time(), B=@n25)) {name=Global Speed, nodeColor=<rgba 0 0 0 1>}
 ]
}
